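# Looping thrust sound cue for the SpaceChimp glider, using shared player glider thrust attenuation.
# A mixer layers a looping thrust bed with quieter flavor one-shots, randomly picked from three waves and retriggered through a looping branch after a randomized delay of a few seconds.
# A continuous modulator at the root scales the mix's volume and pitch from the AnyDirection parameter.
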SoundCue /BRCosmetics/Sounds/Gliders/SpaceChimp/Glider_SpaceChimp_ThrustLoop_Cue
root=ModulatorContinuous_0
nodes [
 {id=Delay_0, children [Random_0], properties {delayMax=5, delayMin=4}}
 {id=Looping_0, children [Delay_0]}
 {id=Mixer_0, children [WavePlayer_0,Looping_0], properties {inputVolume [1,0.4]}}
 {
  id=ModulatorContinuous_0
  children [Mixer_0]
  properties {
   pitchModulationParams {default=0, minOutput=1, parameterName=AnyDirection}
   volumeModulationParams {default=0, minOutput=0.6, parameterName=AnyDirection}
  }
 }
 {id=Random_0, children [WavePlayer_1,WavePlayer_2,WavePlayer_3], properties {weights [1,1,1]}}
 {id=WavePlayer_0, properties {bLooping=true, soundWaveAssetPtr=&/Glider_SpaceChimp_ThrustLoop}}
 {id=WavePlayer_1, properties {soundWaveAssetPtr=&/Glider_SpaceChimp_ThrustFlavor_01}}
 {id=WavePlayer_2, properties {soundWaveAssetPtr=&/Glider_SpaceChimp_ThrustFlavor_02}}
 {id=WavePlayer_3, properties {soundWaveAssetPtr=&/Glider_SpaceChimp_ThrustFlavor_03}}
]
properties {
 attenuationSettings=&Fort_Player_Glider_Thrust_Athena_Att
 audioPropertiesSheet=&APS_Cosmetic_Glider_Loop
 maxDistance=12500
 soundClassObject=&Fort_Gliders_Land
 volumeMultiplier=0.3
}
